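# Masked, default-lit material in the FX folder of the main player's Frolic emote, named for a word element of that effect.
# It exposes a Diffuse texture parameter defaulting to the Frolic gadget texture and a SpecularMasks parameter defaulting to the Fortnite default specular map.
# Usage flags enable it for Niagara sprites, skeletal meshes and mesh deformers.
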Material /Game/Animation/Game/MainPlayer/Emotes/Frolic/FX/M_Emote_Frolic_Word
graph {
 properties {
  bUsedWithMeshDeformer=true
  bUsedWithNiagaraSprites=true
  bUsedWithSkeletalMesh=true
  blendMode=BLEND_Masked
  pixelDepthOffsetMode=PDOM_Legacy
  usedShadingModels=Default Lit
 }
 nodes [
  TextureSampleParameter2D() {parameterName=Diffuse, texture=/Game/Animation/Props/Gadgets_Frolic/Textures/T_Gadget_Frolic_D}
  TextureSampleParameter2D() {parameterName=SpecularMasks, samplerType=SAMPLERTYPE_Masks, texture=&T_Fortnite_Default_S}
 ]
}
